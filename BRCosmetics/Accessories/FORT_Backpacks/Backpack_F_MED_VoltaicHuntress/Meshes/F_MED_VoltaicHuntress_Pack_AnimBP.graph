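# Animation Blueprint for the Voltaic Huntress backpack, adding secondary motion and computing a posture-based pack offset.
# AnimGraph chains root-relative Anim Dynamics with angular springs on left and right tube bones and the zipper, then an additive transform on the pack bone.
# Each update maps owner spine_05 yaw relative to spine_01 from reference pose into Backpack Rotate roll and Backpack Translation Y.
# The transform node shows no visible link to those variables.
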AnimBlueprint /BRCosmetics/Accessories/FORT_Backpacks/Backpack_F_MED_VoltaicHuntress/Meshes/F_MED_VoltaicHuntress_Pack_AnimBP
target_skeleton=&/F_MED_VoltaicHuntress_Pack_Skeleton
graphs [
 AnimGraph {nodes [
   n1 = Anim Dynamics - Bone: r_dyn_tube1(ComponentPose=Local To Component(LocalPose=Input Pose()), Alpha=1, ExternalForce="0,0,0") {
    angularDampingOverride=1
    angularSpringConstant=10000
    bAngularSpring=true
    bOverrideAngularDamping=true
    bUseGravityOverride=true
    gravityScale=0.35
    simulationSpace=RootRelative
    alphaBoolBlend {blendOption=Linear}
    boundBone {boneName=r_dyn_tube1}
    componentLinearAccScale=<1 1 1>
    componentLinearVelScale=<1 1 1>
    gravityOverride {z=-980}
    physicsBodyDefinitions [{
      boundBone {boneName=r_dyn_tube1}
      boxExtents=<15 15 15>
      constraintSetup {angularLimitsMax=<10 0 0>, angularLimitsMin=<-10 -10 0>, angularTarget=<1 0 0>}
      localJointOffset=<0 0 1>
     }]
   }
   n2 = Anim Dynamics - Bone: r_dyn_tube2(ComponentPose=@n1, Alpha=1, ExternalForce="0,0,0") {
    angularDampingOverride=1
    angularSpringConstant=6000
    bAngularSpring=true
    bOverrideAngularDamping=true
    bUseGravityOverride=true
    gravityScale=0.35
    simulationSpace=RootRelative
    alphaBoolBlend {blendOption=Linear}
    boundBone {boneName=r_dyn_tube2}
    componentLinearAccScale=<1 1 1>
    componentLinearVelScale=<1 1 1>
    gravityOverride {z=-980}
    physicsBodyDefinitions [{
      boundBone {boneName=r_dyn_tube2}
      boxExtents=<11 11 11>
      constraintSetup {angularLimitsMax=<10 0 0>, angularLimitsMin=<-10 -10 0>, angularTarget=<1 0 0>}
      localJointOffset=<0 0 1>
     }]
   }
   n3 = Anim Dynamics - Bone: r_dyn_tube3(ComponentPose=@n2, Alpha=1, ExternalForce="0,0,0") {
    angularDampingOverride=1
    angularSpringConstant=4000
    bAngularSpring=true
    bOverrideAngularDamping=true
    bUseGravityOverride=true
    gravityScale=0.35
    simulationSpace=RootRelative
    alphaBoolBlend {blendOption=Linear}
    boundBone {boneName=r_dyn_tube3}
    componentLinearAccScale=<1 1 1>
    componentLinearVelScale=<1 1 1>
    gravityOverride {z=-980}
    physicsBodyDefinitions [{
      boundBone {boneName=r_dyn_tube3}
      boxExtents=<20 20 20>
      constraintSetup {angularLimitsMax=<10 0 0>, angularLimitsMin=<-10 -10 0>, angularTarget=<1 0 0>}
      localJointOffset=<0 0 1>
     }]
   }
   n4 = Anim Dynamics - Bone: l_dyn_tube1(ComponentPose=@n3, Alpha=1, ExternalForce="0,0,0") {
    angularDampingOverride=1
    angularSpringConstant=4000
    bAngularSpring=true
    bOverrideAngularDamping=true
    bUseGravityOverride=true
    gravityScale=0.35
    simulationSpace=RootRelative
    alphaBoolBlend {blendOption=Linear}
    boundBone {boneName=l_dyn_tube1}
    componentLinearAccScale=<1 1 1>
    componentLinearVelScale=<1 1 1>
    gravityOverride {z=-980}
    physicsBodyDefinitions [{
      boundBone {boneName=l_dyn_tube1}
      boxExtents=<20 20 20>
      constraintSetup {angularLimitsMax=<10 0 0>, angularLimitsMin=<-10 -10 0>, angularTarget=<1 0 0>}
      localJointOffset=<0 0 -1>
     }]
   }
   n5 = Anim Dynamics - Bone: l_dyn_tube2(ComponentPose=@n4, Alpha=1, ExternalForce="0,0,0") {
    angularDampingOverride=1
    angularSpringConstant=6000
    bAngularSpring=true
    bOverrideAngularDamping=true
    bUseGravityOverride=true
    gravityScale=0.35
    simulationSpace=RootRelative
    alphaBoolBlend {blendOption=Linear}
    boundBone {boneName=l_dyn_tube2}
    componentLinearAccScale=<1 1 1>
    componentLinearVelScale=<1 1 1>
    gravityOverride {z=-980}
    physicsBodyDefinitions [{
      boundBone {boneName=l_dyn_tube2}
      boxExtents=<11 11 11>
      constraintSetup {angularLimitsMax=<10 0 0>, angularLimitsMin=<-10 -10 0>, angularTarget=<1 0 0>}
      localJointOffset=<0 0 -1>
     }]
   }
   n6 = Anim Dynamics - Bone: l_dyn_tube3(ComponentPose=@n5, Alpha=1, ExternalForce="0,0,0") {
    angularDampingOverride=1
    angularSpringConstant=10000
    bAngularSpring=true
    bOverrideAngularDamping=true
    bUseGravityOverride=true
    gravityScale=0.35
    simulationSpace=RootRelative
    alphaBoolBlend {blendOption=Linear}
    boundBone {boneName=l_dyn_tube3}
    componentLinearAccScale=<1 1 1>
    componentLinearVelScale=<1 1 1>
    gravityOverride {z=-980}
    physicsBodyDefinitions [{
      boundBone {boneName=l_dyn_tube3}
      boxExtents=<15 15 15>
      constraintSetup {angularLimitsMax=<10 0 0>, angularLimitsMin=<-10 -10 0>, angularTarget=<1 0 0>}
      localJointOffset=<0 0 -1>
     }]
   }
   n7 = Anim Dynamics - Bone: dyn_zipper(ComponentPose=@n6, Alpha=1) {
    angularDampingOverride=1
    angularSpringConstant=20000
    bAngularSpring=true
    bOverrideAngularDamping=true
    bUseGravityOverride=true
    gravityScale=0.35
    simulationSpace=RootRelative
    alphaBoolBlend {blendOption=Linear}
    boundBone {boneName=dyn_zipper}
    componentLinearAccScale=<1 1 1>
    componentLinearVelScale=<1 1 1>
    gravityOverride {z=-980}
    physicsBodyDefinitions [{
      boundBone {boneName=dyn_zipper}
      boxExtents=<20 30 30>
      constraintSetup {angularLimitsMax=<15 15 15>, angularLimitsMin=<-15 -15 -15>, angularTarget=<1 0 0>}
      localJointOffset=<4 0 0>
     }]
   }
   n8 = "Transform (Modify) Bone - Bone: pack"(ComponentPose=@n7, Translation="0,0,0", Rotation="0,0,0") {rotationMode=BMM_Additive, translationMode=BMM_Additive, alphaBoolBlend {blendOption=Linear}, boneToModify {boneName=pack}}
   Output Pose(Result=Component To Local(ComponentPose=@n8))
  ]}
 EventGraph {nodes [
   Event BlueprintUpdateAnimation()
   n1 = GetOwnerSkeletalMeshComponent()
   n2 = BreakRotator(InRot=BreakTransform(InTransform=GetDeltaTransformFromRefPose(self=@n1, BoneName=spine_05, BaseName=spine_01)).Rotation)
   > Set Backpack Rotate(Backpack Rotate:Rotator=MakeRotator(Roll=MapRangeClamped(Value=@n2.Yaw, InRangeA=0, InRangeB=-30, OutRangeB=35))) -> (Output_Get:Rotator)
   > Set Backpack Translation(Backpack Translation:Vector=MakeVector(Y=MapRangeClamped(Value=@n2.Yaw, InRangeA=0, InRangeB=-30, OutRangeB=-10))) -> (Output_Get:Vector)
   TryGetPawnOwner()
  ]}
]
variables [
 Backpack Rotate {type=Rotator}
 Backpack Translation {type=Vector}
]
